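# Physics asset for the emote face-accessory mesh of the PastaSauceSpice medium female cosmetic body, covering two dynamic hair strands.
# Left and right four-bone capsule chains hang from small master bodies linked to the neck, with tight twist and modest swing limits per segment.
# Capsules on upper spine, neck and arms approximate the body, likely as hair colliders.
# Constraints among these body capsules and to the hair masters leave swing and twist free.
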PhysicsAsset /BRCosmetics/Characters/Player/Female/Medium/Bodies/F_MED_PastaSauceSpice/Meshes/Parts/F_MED_PastaSauceSpice_FaceAcc_Emote_Physics
bodies {
 dyn_hair_l_01 [{shapeType=Capsule, length=8, radius=2, shapeName=dyn_hair_l_01_capsule, center=<1.5 1 0>, rotation=<rot 90 0 0>}]
 dyn_hair_l_02 [{shapeType=Capsule, length=8, radius=2, shapeName=dyn_hair_l_02_capsule, center=<1.5 1 0>, rotation=<rot 90 0 0>}]
 dyn_hair_l_03 [{shapeType=Capsule, length=8, radius=2, shapeName=dyn_hair_l_03_capsule, center=<1.5 1 0>, rotation=<rot 90 74.6511 69.651>}]
 dyn_hair_l_04 [{shapeType=Capsule, length=20, radius=2, shapeName=dyn_hair_l_04_capsule, center=<4 0.85 0>, rotation=<rot 90 0 0>}]
 dyn_hair_l_MSTR [{shapeType=Capsule, length=0.505, radius=0.505, shapeName=dyn_hair_l_MSTR_capsule, rotation=<rot 0 -90 0>}]
 dyn_hair_r_01 [{shapeType=Capsule, length=8, radius=2, shapeName=dyn_hair_r_01_capsule, center=<1.5 -1 0>, rotation=<rot 90 0 0>}]
 dyn_hair_r_02 [{shapeType=Capsule, length=8, radius=2, shapeName=dyn_hair_r_02_capsule, center=<1.5 -1 0>, rotation=<rot 90 0 0>}]
 dyn_hair_r_03 [{shapeType=Capsule, length=8, radius=2, shapeName=dyn_hair_r_03_capsule, center=<1.5 -1 0>, rotation=<rot 90 13.1035 18.1036>}]
 dyn_hair_r_04 [{shapeType=Capsule, length=20, radius=2, shapeName=dyn_hair_r_04_capsule, center=<4 -0.85 0>, rotation=<rot 90 0 0>}]
 dyn_hair_r_MSTR [{shapeType=Capsule, length=0.505, radius=0.505, shapeName=dyn_hair_r_MSTR_capsule, rotation=<rot 0 -90 0>}]
 lowerarm_l [{shapeType=Capsule, length=30, radius=6, center=<10 0 0>, rotation=<rot 90 0 0>}]
 lowerarm_r [{shapeType=Capsule, length=30, radius=6, center=<-10 0 0>, rotation=<rot 90 0 180>}]
 neck_01 [
  {shapeType=Capsule, length=26.5, radius=5.755, shapeName=neck_01_capsule, center=<-0.0409 -2.4058 0>, rotation=<rot 0 -89.9999 0>}
  {shapeType=Capsule, length=6, radius=4, shapeName=neck_01_capsule, center=<3 0 0>, rotation=<rot 90 -10 0>}
 ]
 spine_04 [{shapeType=Capsule, length=4, radius=10.5, center=<0 1.9256 -0.0002>}]
 spine_05 [{shapeType=Capsule, length=9, radius=9, center=<2 0.2464 -0.0002>}]
 upperarm_l [{shapeType=Capsule, length=24, radius=6, center=<8 0 0>, rotation=<rot 90 0 0>}]
 upperarm_r [{shapeType=Capsule, length=24, radius=6, center=<-8 0 0>, rotation=<rot 90 0 180>}]
}
constraints [
 {
  bone1Name=spine_05
  bone2Name=spine_04
  swing1LimitDegrees=45
  swing1Motion=Free
  swing2LimitDegrees=45
  swing2Motion=Free
  twistLimitDegrees=45
  twistMotion=Free
 }
 {
  bone1Name=upperarm_l
  bone2Name=spine_05
  swing1LimitDegrees=45
  swing1Motion=Free
  swing2LimitDegrees=45
  swing2Motion=Free
  twistLimitDegrees=45
  twistMotion=Free
 }
 {
  bone1Name=lowerarm_l
  bone2Name=upperarm_l
  swing1LimitDegrees=45
  swing1Motion=Free
  swing2LimitDegrees=45
  swing2Motion=Free
  twistLimitDegrees=45
  twistMotion=Free
 }
 {
  bone1Name=upperarm_r
  bone2Name=spine_05
  swing1LimitDegrees=45
  swing1Motion=Free
  swing2LimitDegrees=45
  swing2Motion=Free
  twistLimitDegrees=45
  twistMotion=Free
 }
 {
  bone1Name=lowerarm_r
  bone2Name=upperarm_r
  swing1LimitDegrees=45
  swing1Motion=Free
  swing2LimitDegrees=45
  swing2Motion=Free
  twistLimitDegrees=45
  twistMotion=Free
 }
 {
  bone1Name=neck_01
  bone2Name=spine_05
  swing1LimitDegrees=45
  swing1Motion=Free
  swing2LimitDegrees=45
  swing2Motion=Free
  twistLimitDegrees=45
  twistMotion=Free
 }
 {
  bone1Name=dyn_hair_l_MSTR
  bone2Name=neck_01
  swing1LimitDegrees=45
  swing1Motion=Free
  swing2LimitDegrees=45
  swing2Motion=Free
  twistLimitDegrees=45
  twistMotion=Free
 }
 {bone1Name=dyn_hair_l_01, bone2Name=dyn_hair_l_MSTR, swing1LimitDegrees=30, swing2LimitDegrees=16, twistLimitDegrees=3}
 {bone1Name=dyn_hair_l_02, bone2Name=dyn_hair_l_01, swing1LimitDegrees=20, swing2LimitDegrees=16, twistLimitDegrees=3}
 {bone1Name=dyn_hair_l_03, bone2Name=dyn_hair_l_02, swing1LimitDegrees=20, swing2LimitDegrees=16, twistLimitDegrees=3}
 {bone1Name=dyn_hair_l_04, bone2Name=dyn_hair_l_03, swing1LimitDegrees=20, swing2LimitDegrees=16, twistLimitDegrees=3}
 {
  bone1Name=dyn_hair_r_MSTR
  bone2Name=neck_01
  swing1LimitDegrees=45
  swing1Motion=Free
  swing2LimitDegrees=45
  swing2Motion=Free
  twistLimitDegrees=45
  twistMotion=Free
 }
 {bone1Name=dyn_hair_r_01, bone2Name=dyn_hair_r_MSTR, swing1LimitDegrees=30, swing2LimitDegrees=16, twistLimitDegrees=3}
 {bone1Name=dyn_hair_r_02, bone2Name=dyn_hair_r_01, swing1LimitDegrees=20, swing2LimitDegrees=16, twistLimitDegrees=3}
 {bone1Name=dyn_hair_r_03, bone2Name=dyn_hair_r_02, swing1LimitDegrees=20, swing2LimitDegrees=16, twistLimitDegrees=3}
 {bone1Name=dyn_hair_r_04, bone2Name=dyn_hair_r_03, swing1LimitDegrees=20, swing2LimitDegrees=16, twistLimitDegrees=3}
]
