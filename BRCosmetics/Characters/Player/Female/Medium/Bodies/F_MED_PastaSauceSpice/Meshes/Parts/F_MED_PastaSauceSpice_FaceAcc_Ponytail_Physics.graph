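# Physics asset for the ponytail face accessory of the F_MED_PastaSauceSpice female player cosmetic, setting up the bodies and joints for hair physics.
# A five-capsule hair chain hangs from a small base body linked to the neck, with limited swing and tight twist between segments.
# Capsules on the pelvis, spine, neck, arms and thighs give the body volume, joined by constraints with free swing and twist.
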PhysicsAsset /BRCosmetics/Characters/Player/Female/Medium/Bodies/F_MED_PastaSauceSpice/Meshes/Parts/F_MED_PastaSauceSpice_FaceAcc_Ponytail_Physics
bodies {
 dyn_hair_a [{shapeType=Capsule, length=8, radius=5.5, center=<10 0 0>, rotation=<rot 90 0 0>}]
 dyn_hair_a_base [{shapeType=Capsule, length=1, radius=1}]
 dyn_hair_b [{shapeType=Capsule, length=7.75, radius=5.5, center=<9 0 0>, rotation=<rot 90 0 0>}]
 dyn_hair_c [{shapeType=Capsule, length=9.75, radius=5.5, center=<11 0 0>, rotation=<rot 90 0 0>}]
 dyn_hair_d [{shapeType=Capsule, length=8.75, radius=5.5, center=<10 0 0>, rotation=<rot 90 0 0>}]
 dyn_hair_e [{shapeType=Capsule, length=14.75, radius=5.5, center=<12 0 0>, rotation=<rot 90 0 0>}]
 lowerarm_l [{shapeType=Capsule, length=30, radius=16, center=<10 -12 6>, rotation=<rot 90 0 0>}]
 lowerarm_r [{shapeType=Capsule, length=30, radius=16, center=<-10 12 -6>, rotation=<rot 90 0 180>}]
 neck_01 [{shapeType=Capsule, length=26.5, radius=5.755, shapeName=neck_01_capsule, center=<-0.0409 -2.4058 0>, rotation=<rot 0 -89.9999 0>}]
 pelvis [{shapeType=Capsule, length=35.5613, radius=9.5522, center=<-1.0276 0.7357 0>}]
 spine_02 [{shapeType=Capsule, length=12.505, radius=9.255, shapeName=spine_02_capsule, rotation=<rot 0 -90 0>}]
 spine_04 [{shapeType=Capsule, length=14.75, radius=10.5, center=<-0.1575 1.7707 -0.0002>}]
 spine_05 [{shapeType=Capsule, length=24.25, radius=9, center=<0.3742 0.2464 -0.0002>}]
 thigh_l [{shapeType=Capsule, length=39.8074, radius=13.2812, center=<-21.9558 -0.0131 -0.1928>, rotation=<rot 90 0 0>}]
 thigh_r [{shapeType=Capsule, length=39.8074, radius=13.2812, center=<21.9558 0.0131 0.1928>, rotation=<rot 90 0 180>}]
 upperarm_l [{shapeType=Capsule, length=24, radius=12, center=<8 -6 5>, rotation=<rot 90 0 0>}]
 upperarm_r [{shapeType=Capsule, length=24, radius=12, center=<-8 6 -5>, rotation=<rot 90 0 180>}]
}
constraints [
 {
  bone1Name=spine_02
  bone2Name=pelvis
  swing1LimitDegrees=45
  swing1Motion=Free
  swing2LimitDegrees=45
  swing2Motion=Free
  twistLimitDegrees=45
  twistMotion=Free
 }
 {
  bone1Name=spine_05
  bone2Name=spine_04
  swing1LimitDegrees=45
  swing1Motion=Free
  swing2LimitDegrees=45
  swing2Motion=Free
  twistLimitDegrees=45
  twistMotion=Free
 }
 {
  bone1Name=upperarm_l
  bone2Name=spine_05
  swing1LimitDegrees=45
  swing1Motion=Free
  swing2LimitDegrees=45
  swing2Motion=Free
  twistLimitDegrees=45
  twistMotion=Free
 }
 {
  bone1Name=lowerarm_l
  bone2Name=upperarm_l
  swing1LimitDegrees=45
  swing1Motion=Free
  swing2LimitDegrees=45
  swing2Motion=Free
  twistLimitDegrees=45
  twistMotion=Free
 }
 {
  bone1Name=upperarm_r
  bone2Name=spine_05
  swing1LimitDegrees=45
  swing1Motion=Free
  swing2LimitDegrees=45
  swing2Motion=Free
  twistLimitDegrees=45
  twistMotion=Free
 }
 {
  bone1Name=lowerarm_r
  bone2Name=upperarm_r
  swing1LimitDegrees=45
  swing1Motion=Free
  swing2LimitDegrees=45
  swing2Motion=Free
  twistLimitDegrees=45
  twistMotion=Free
 }
 {
  bone1Name=neck_01
  bone2Name=spine_05
  swing1LimitDegrees=45
  swing1Motion=Free
  swing2LimitDegrees=45
  swing2Motion=Free
  twistLimitDegrees=45
  twistMotion=Free
 }
 {
  bone1Name=dyn_hair_a_base
  bone2Name=neck_01
  swing1LimitDegrees=45
  swing1Motion=Free
  swing2LimitDegrees=45
  swing2Motion=Free
  twistLimitDegrees=45
  twistMotion=Free
 }
 {bone1Name=dyn_hair_a, bone2Name=dyn_hair_a_base, swing1LimitDegrees=30, swing2LimitDegrees=20, twistLimitDegrees=5}
 {bone1Name=dyn_hair_b, bone2Name=dyn_hair_a, swing1LimitDegrees=30, swing2LimitDegrees=20, twistLimitDegrees=5}
 {bone1Name=dyn_hair_c, bone2Name=dyn_hair_b, swing1LimitDegrees=30, swing2LimitDegrees=30, twistLimitDegrees=5}
 {bone1Name=dyn_hair_d, bone2Name=dyn_hair_c, swing1LimitDegrees=30, swing2LimitDegrees=30, twistLimitDegrees=5}
 {bone1Name=dyn_hair_e, bone2Name=dyn_hair_d, swing1LimitDegrees=30, swing2LimitDegrees=30, twistLimitDegrees=5}
 {
  bone1Name=thigh_l
  bone2Name=pelvis
  swing1LimitDegrees=45
  swing1Motion=Free
  swing2LimitDegrees=45
  swing2Motion=Free
  twistLimitDegrees=45
  twistMotion=Free
 }
 {
  bone1Name=thigh_r
  bone2Name=pelvis
  swing1LimitDegrees=45
  swing1Motion=Free
  swing2LimitDegrees=45
  swing2Motion=Free
  twistLimitDegrees=45
  twistMotion=Free
 }
 {
  bone1Name=spine_04
  bone2Name=spine_02
  swing1LimitDegrees=45
  swing1Motion=Free
  swing2LimitDegrees=45
  swing2Motion=Free
  twistLimitDegrees=45
  twistMotion=Free
 }
]
